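# Parent material for Fridge SE vehicle wheels, adding animated glowing rim and tire effects to MF_WheelParent_Simple.
# Mask textures and reroutes adjust base color, metallic and roughness, and build Custom Color emissive from sine pulses plus tire hex tiles offset by Hex Offset.
# A masked sine pulse drives local-space vertex offset; vertex color channels are exposed as masks.
# The result passes through shared vehicle damage, then de-rez transition functions.
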Material /VehicleCosmetics/Wheels/Fridge_SE/Materials/M_WheelParent_Fridge_SE
graph {
 properties {
  bAutomaticallySetUsageInEditor=false
  bUseMaterialAttributes=true
  bUsedWithMeshDeformer=true
  bUsedWithSkeletalMesh=true
  pixelDepthOffsetMode=PDOM_Legacy
  usedShadingModels=Default Lit
 }
 nodes [
  n1 = GetMaterialAttributes(None=MaterialFunctionCall() {materialFunction=&MF_WheelParent_Simple}.MaterialAttributes) {attributeGetTypes [#1, #2, #3]}
  n2 = NamedRerouteUsage()
  n3 = NamedRerouteUsage()
  n4 = NamedRerouteUsage()
  n5 = SetMaterialAttributes!(
   MaterialAttributes=@n1.MaterialAttributes
   Base Color=LinearInterpolate(A=LinearInterpolate(A=@n1.Base Color, B=@n2, Alpha=@n3), Alpha=@n4) {constB=0}
   Metallic=LinearInterpolate(A=@n1.Metallic, Alpha=@n4)
   Roughness=LinearInterpolate(A=@n1.Roughness, Alpha=@n4)
   Emissive Color=NamedRerouteUsage()
   World Position Offset=NamedRerouteUsage()
  ) {attributeSetTypes [
    #1
    #2
    #3
    #4
    #5
   ]}
  MaterialFunctionCall(In=MaterialFunctionCall(Material=@n5) {materialFunction=&MF_BaseVehicle_Damage}.Result) {materialFunction=&MF_VehicleDeRez_Transition}
  n6 = VertexColor()
  NamedRerouteDeclaration(None=@n6.R) {name=VC_R, nodeColor=<rgba 1 0 0 1>}
  n7 = TextureSampleParameter2D() {group=1_Input Textures, parameterName=Rim Masks, texture=/VehicleCosmetics/Wheels/Fridge_SE/Textures/T_Fridge_SE_Rim_M}
  n8 = OneMinus(None=NamedRerouteUsage())
  n9 = Subtract(A=LinearInterpolate(B=@n7.G, Alpha=@n8), B=NamedRerouteUsage())
  n10 = NamedRerouteDeclaration(None=VectorParameter() {group=5_Special Effects, parameterName=Custom Color, defaultValue=<rgba 0 0.8714 0.0185 1>}.RGB) {name=Custom Color, nodeColor=<rgba 1 0.5 0 1>}
  n11 = Panner(Coordinate=TextureCoordinate() {vTiling=3}) {speedY=1}
  n12 = ConstantBiasScale(None=Sine(None=ComponentMask(None=@n11) {g=true})) {bias=-0.1, scale=1}
  n13 = LinearInterpolate(
   A=Multiply(A=NamedRerouteDeclaration(None=@n9) {name=Masks for BC, nodeColor=<rgba 1 0 0.9141 1>}, B=@n10)
   B=@n10
   Alpha=Multiply(A=Clamp(None=@n12), B=@n7.B)
  )
  n14 = ScalarParameter() {defaultValue=3, group=5_Special Effects, parameterName=Rim Emissive}
  n15 = TextureSampleParameter2D() {
   group=1_Input Textures
   parameterName=Tire Masks
   samplerSource=SSM_Wrap_WorldGroupSettings
   texture=/VehicleCosmetics/Wheels/Fridge_SE/Textures/T_Fridge_SE_Tire_M
  }
  n16 = ComponentMask(None=Panner(Coordinate=TextureCoordinate()) {speedX=1}) {r=true}
  n17 = Clamp(None=ConstantBiasScale(None=Sine(None=@n16)) {bias=-0.1, scale=1})
  n18 = TextureSampleParameter2D(UVs=Panner(Coordinate=TextureCoordinate() {coordinateIndex=1, uTiling=9, vTiling=2}, Time=ScalarParameter() {parameterName=Hex Offset}) {speedY=1}) {group=5_Special Effects, parameterName=WPO Mask, texture=/VehicleCosmetics/Wheels/Fridge_SE/Textures/Wheel_Fridge_ARC_Tiles_FieldHex00_RGB}
  n19 = Multiply(A=@n18.B) {constB=12}
  n20 = TextureSampleParameter2D(UVs=TextureCoordinate() {coordinateIndex=1}) {
   group=1_Input Textures
   parameterName=Tire Masks
   samplerSource=SSM_Wrap_WorldGroupSettings
   texture=&T_Fridge_SE_Tire_M
  }
  n21 = LinearInterpolate(B=@n10, Alpha=LinearInterpolate(A=Multiply(A=@n15.G, B=@n17), B=Add(A=@n19) {constB=0.07}, Alpha=@n20.B))
  n22 = ScalarParameter() {defaultValue=6, group=5_Special Effects, parameterName=Tire Emissive}
  NamedRerouteDeclaration(None=LinearInterpolate(A=Multiply(A=@n13, B=@n14), B=Multiply(A=@n21, B=@n22), Alpha=NamedRerouteUsage())) {name=Emissive, nodeColor=<rgba 1 0.4453 0 1>}
  n23 = ComponentMask(None=Panner(Coordinate=TextureCoordinate()) {speedX=0.3}) {r=true}
  n24 = Clamp(None=ConstantBiasScale(None=Sine(None=@n23)) {scale=0.2})
  n25 = TextureCoordinate() {vTiling=3}
  n26 = ComponentMask(None=Panner(Coordinate=Add(A=@n25) {constB=-1}, Time=Constant() {r=0.2}) {speedX=1, speedY=1}) {g=true}
  n27 = Clamp(None=Ceil(None=OneMinus(None=@n26)))
  n28 = Clamp(None=LinearInterpolate(B=Multiply(A=@n24, B=@n27), Alpha=NamedRerouteUsage()))
  n29 = AppendVector(A=AppendVector(A=@n28, B=@n28), B=@n28)
  n30 = VectorParameter() {group=5_Special Effects, parameterName=WPO, defaultValue=<rgba 1 0 1 1>}
  NamedRerouteDeclaration(None=Transform(None=Multiply(A=@n29, B=@n30.RGB)) {transformSourceType=TRANSFORMSOURCE_Local}) {name=WPO, nodeColor=<rgba 0.125 0 1 1>}
  NamedRerouteDeclaration(None=@n6.G) {name=VC_G, nodeColor=<rgba 0 1 0 1>}
  NamedRerouteDeclaration(None=Multiply(A=@n7.R, B=@n8)) {name=Black Mask, nodeColor=<rgba 0.6406 1 0 1>}
 ]
}
