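# Translucent unlit energy material for the Fridge_SE vehicle wheel, with transition and unlock effects.
# Opacity blends an animated hex lock-back pattern with panning noise streaks shaped by a UV edge and power-gradient mask.
# Paint Color drives emissive, and Unlock FX scales a sine/cosine World Position Offset wobble.
# Vertex-color channels and the combined Transition/Unlock value are exposed as named reroutes alongside the main outputs.
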Material /VehicleCosmetics/Wheels/Fridge_SE/Materials/M_Wheel_Fridge_Energy
graph {
 properties {
  bAutomaticallySetUsageInEditor=false
  bUseMaterialAttributes=true
  bUsedWithMeshDeformer=true
  bUsedWithSkeletalMesh=true
  blendMode=BLEND_Translucent
  shadingModel=MSM_Unlit
  usedShadingModels=Unlit
 }
 nodes [
  n1 = VertexColor()
  NamedRerouteDeclaration(None=@n1.R) {name=VC_R, nodeColor=<rgba 1 0.1641 0 1>}
  NamedRerouteDeclaration(None=@n1.G) {name=VC_G, nodeColor=<rgba 0 1 0 1>}
  n2 = TextureSample() {texture=/VehicleCosmetics/Wheels/_SharedTextures/FX/HexMask}
  n3 = Time()
  n4 = Clamp(None=Sine(None=Add(A=@n2.R, B=@n3)) {period=0.4})
  n5 = TextureSample(UVs=TextureCoordinate() {uTiling=1.2, vTiling=1.2}) {texture=/VehicleCosmetics/Wheels/_SharedTextures/FX/LockBackHex}
  n6 = Power(Base=@n2.R) {constExponent=4}
  n7 = Add(A=Multiply(A=Multiply(A=@n4, B=@n2.R), B=@n5.A), B=Multiply(A=Multiply(A=@n6, B=@n5.A)) {constB=12})
  n8 = Multiply(A=@n2.R, B=@n5.RGB)
  NamedRerouteDeclaration(None=LinearInterpolate(B=Add(A=@n7, B=@n8), Alpha=NamedRerouteUsage()) {constA=0.2}) {name=Back OP, nodeColor=<rgba 0.7578 0 1 1>}
  n9 = ScalarParameter() {defaultValue=0.6, group=00 Base, parameterName=Unlock FX Int}
  n10 = Multiply(A=Time(), B=ScalarParameter() {defaultValue=12, group=00 Base, parameterName=Unlock FX Speed})
  n11 = Sine(None=@n10)
  n12 = Multiply(A=AppendVector(A=AppendVector(A=@n9, B=@n9), B=@n9), B=AppendVector(A=AppendVector(A=@n11, B=@n11), B=Cosine(None=@n10)))
  n13 = ScalarParameter() {group=00 Base, parameterName=Unlock FX}
  NamedRerouteDeclaration(None=LinearInterpolate(A=Multiply(A=@n12, B=@n13), Alpha=NamedRerouteUsage()) {constB=0}) {name=WPO, nodeColor=<rgba 0 1 0.4375 1>}
  n14 = VectorParameter() {group=00 Base, parameterName=Paint Color, sortPriority=0, defaultValue=<rgba 1 0.0999 0.147 1>}
  n15 = ScalarParameter() {group=00 Base, parameterName=Unlock FX}
  n16 = LinearInterpolate(A=ScalarParameter() {group=00 Base, parameterName=Transition}, B=@n15, Alpha=@n15)
  n17 = LinearInterpolate(A=Multiply(A=@n14.RGB, B=@n14.RGB), B=Multiply(A=@n14.RGB, B=@n14.RGB), Alpha=@n16)
  NamedRerouteDeclaration(None=LinearInterpolate(A=@n17, B=Multiply(A=@n17) {constB=6}, Alpha=NamedRerouteUsage())) {name=Emissive, nodeColor=<rgba 1 0.1172 0 1>}
  NamedRerouteDeclaration(None=@n16) {name=Trans/Unlock, nodeColor=<rgba 0.1953 0 1 1>}
  n18 = NamedRerouteUsage()
  n19 = TextureSample(UVs=Panner(Coordinate=TextureCoordinate() {uTiling=0.2, vTiling=0.04}) {speedY=0.1}) {samplerSource=SSM_Wrap_WorldGroupSettings, texture=/VehicleCosmetics/Wheels/_SharedTextures/FX/Noise_Generic_01_Pack}
  n20 = TextureCoordinate() {uTiling=0.6, vTiling=0}
  n21 = TextureSample(UVs=Panner(Coordinate=@n20) {speedY=0.03}) {samplerSource=SSM_Wrap_WorldGroupSettings, texture=/VehicleCosmetics/Wheels/_SharedTextures/FX/Noise_Macro_01_Pack}
  n22 = TextureSample(UVs=Panner(Coordinate=@n20) {speedX=0.01, speedY=0.02}) {samplerSource=SSM_Wrap_WorldGroupSettings, texture=&Noise_Macro_01_Pack}
  n23 = Multiply(A=Add(A=Multiply(A=@n19.B) {constB=6}, B=Add(A=@n21.R, B=@n22.R)))
  n24 = TextureCoordinate()
  n25 = ComponentMask(None=Add(A=@n24) {constB=-0.5}) {r=true}
  n26 = Subtract(A=OneMinus(None=Abs(None=@n25))) {constB=0.5}
  n27 = ScalarParameter() {defaultValue=20, group=00 Base, parameterName=Unlock Beg Power}
  n28 = ScalarParameter() {defaultValue=5, group=00 Base, parameterName=Unlock End Power}
  n29 = Clamp(None=Multiply(A=NamedRerouteUsage()) {constB=5})
  n30 = Multiply(
   A=Clamp(None=Multiply(A=@n26) {constB=16})
   B=Power(Base=ComponentMask(None=@n24) {g=true}, Exp=LinearInterpolate(A=@n27, B=@n28, Alpha=@n29))
  )
  n31 = Multiply(A=Clamp(None=Multiply(A=@n23, B=@n30)), B=ScalarParameter() {defaultValue=1, group=00 Base, parameterName=Energy Opacity})
  n32 = NamedRerouteUsage()
  NamedRerouteDeclaration(None=Clamp(None=LinearInterpolate(A=@n18, B=@n31, Alpha=@n32))) {name=Opacity, nodeColor=<rgba 0.125 1 0 1>}
  SetMaterialAttributes!(MaterialAttributes, Emissive Color=NamedRerouteUsage(), Opacity=NamedRerouteUsage(), World Position Offset=NamedRerouteUsage()) {attributeSetTypes [#1, #2, #3]}
 ]
}
